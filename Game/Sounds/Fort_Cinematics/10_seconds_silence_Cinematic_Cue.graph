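# Sound cue for the Fort cinematics audio set that plays a silent wave, named as ten seconds of silence.
# A single looping wave player serves as the root, so the silence repeats continuously.
# The cue is assigned to the Fort_Cinematics_PSM sound class and has a very large max distance.
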SoundCue /Game/Sounds/Fort_Cinematics/10_seconds_silence_Cinematic_Cue
root=WavePlayer_0
nodes [{id=WavePlayer_0, properties {bLooping=true, soundWaveAssetPtr=&/10_seconds_silence}}]
properties {maxDistance=2097152, soundClassObject=/Game/Sounds/FortSoundClassesAndModes/Fort_Cinematics_PSM, volumeMultiplier=1}
